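# Animation Blueprint for the Dark Paintballer backpack cosmetic that shifts and tilts the pack as the wearer's spine twists.
# On initialize it caches the owner's skeletal mesh component.
# Each update it turns the quaternion twist delta between spine_01 and spine_05, relative to the reference pose, into an alpha.
# That alpha blends an additive Transform Bone on backpack_root over the input pose, offsetting and rotating the pack.
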AnimBlueprint /BRCosmetics/Accessories/FORT_Backpacks/M_MED_Dark_Paintballer_Backpack/Meshes/SK_Backpack_Dark_Paintballer_AnimBP
target_skeleton=&/SK_Backpack_Dark_Paintballer_Skeleton
graphs [
 AnimGraph {nodes [
   n1 = "Transform (Modify) Bone - Bone: backpack_root"(
    ComponentPose=Local To Component(LocalPose=Input Pose())
    Alpha=Get backpack_down_rotation() -> (backpack_down_rotation:double)
    Translation="0,-5,0"
    Rotation="-10,-10,35"
    Scale="1,1,1"
   ) {
    rotationMode=BMM_Additive
    rotationSpace=BCS_ParentBoneSpace
    translationMode=BMM_Additive
    translationSpace=BCS_BoneSpace
    alphaBoolBlend {blendOption=Linear}
    boneToModify {boneName=backpack_root}
    rotation=<rot -10 -10 35>
    translation {y=-5}
   }
   Output Pose(Result=Component To Local(ComponentPose=@n1)) {layerGroup=None}
  ]}
 EventGraph {nodes [
   Event BlueprintUpdateAnimation()
   > Set backpack_down_rotation(backpack_down_rotation:double=GetAlphaFromReferenceTransformDelta(
     SkeletalMeshComponent=Get parent_skeletal_mesh() -> (parent_skeletal_mesh:Skeletal Mesh Component Object Reference)
     BoneName=spine_05
     BaseName=spine_01
     DeltaComponentType=QuaternionTwist
     InputRange=(X=0,Y=45)
     TwistAxis="0,0,1"
    )) -> (Output_Get:double)
   Event BlueprintInitializeAnimation()
   > Set parent_skeletal_mesh(parent_skeletal_mesh:Skeletal Mesh Component Object Reference=GetOwnerSkeletalMeshComponent()) -> (Output_Get:Skeletal Mesh Component Object Reference)
  ]}
]
variables [
 backpack_down_rotation {type=number}
 parent_skeletal_mesh {type=SkeletalMeshComponent}
]
